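# DrawWaterInfo is an unlit Water material, apparently for drawing water info data, with negative emissive allowed and spline and instanced mesh usage.
# A Custom node decodes vertex color red as speed scaled by MaxFlowVelocity and green as angle, then remaps the XY flow vector to 0–1.
# That flow is appended to texture coordinates and computed per vertex through a VertexInterpolator; no connection to material outputs is shown.
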Material /Water/Materials/WaterInfo/DrawWaterInfo
graph {
 properties {
  bAllowNegativeEmissiveColor=true
  bUsedWithInstancedStaticMeshes=true
  bUsedWithSplineMeshes=true
  bUsedWithStaticLighting=true
  bUsedWithWater=true
  pixelDepthOffsetMode=PDOM_Legacy
  shadingModel=MSM_Unlit
  usedShadingModels=Unlit
 }
 nodes [
  MaterialExpressionComponentMask_0 = ComponentMask(None=VertexColor()) {g=true, r=true}
  MaterialExpressionScalarParameter_5 = ScalarParameter() {defaultValue=1024, parameterName=MaxFlowVelocity}
  n1 = Custom!(Flow=@MaterialExpressionComponentMask_0, MaxVelocity=@MaterialExpressionScalarParameter_5) {
   code="float Magnitude = Flow.x * MaxVelocity;\r\nfloat Direction = Flow.y * 2 * PI;\r\n\r\nfloat FlowX = Magnitude * cos(Direction);\r\nfloat FlowY = Magnitude * sin(Direction);\r\n\r\nfloat NormalizedFlowX = (FlowX / (2 * MaxVelocity)) + 0.5;\r\nfloat NormalizedFlowY = (FlowY / (2 * MaxVelocity)) + 0.5;\r\n\r\nreturn float2(NormalizedFlowX, NormalizedFlowY);"
   description=Normalize Flow
   outputType=CMOT_Float2
   inputs [
    {inputName=Flow, input {expression=&self:MaterialExpressionComponentMask_0}}
    {inputName=MaxVelocity, input {expression=&self:MaterialExpressionScalarParameter_5}}
   ]
  }
  VertexInterpolator(VS=AppendVector(A=@n1, B=TextureCoordinate()))
 ]
}
